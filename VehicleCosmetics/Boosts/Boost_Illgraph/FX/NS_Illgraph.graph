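# Niagara system for the Illgraph vehicle boost cosmetic, emitting trails and bursts from the vehicle's skeletal-mesh boost sockets.
# Ribbon and Twist emitters layer six and four ribbon renderers, with per-spawn-group widths, setting dissolve UV and world-position-offset parameters respectively.
# The Sprite emitter uses four spawn groups to mix flickering meshes and animated subimage sprites.
# Mask modules and bIsFrontendPreview retune values per game mode and preview; painted-variant overrides are disabled.
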NiagaraSystem /VehicleCosmetics/Boosts/Boost_Illgraph/FX/NS_Illgraph
dependencies {
 usedDataInterfaces [
  &NiagaraDataInterfaceCurve
  &NiagaraDataInterfaceArrayFloat
  &NiagaraDataInterfaceArrayInt32
  &NiagaraDataInterfaceSkeletalMesh
  &NiagaraDataInterfaceArrayFloat3
 ]
 usedDynamicInputs [&SelectVectorFromArray]
 usedModules [
  &self:SystemSpawnScript.SystemScriptSource.SystemScriptGraph.NiagaraNodeAssignment_3.SetVariables_#1
  &self:SystemSpawnScript.SystemScriptSource.SystemScriptGraph.NiagaraNodeAssignment_4.SetVariables_#2
  &self:SystemSpawnScript.SystemScriptSource.SystemScriptGraph.NiagaraNodeAssignment_2.SetVariables_#3
  &SystemState
 ]
}
emitters {
 NE_VC_Boost_Ribbon {
  emitterUpdateScript [
   &EmitterState()
   &VC_FilteredSocketCount()
   &VC_RateQualityHandler(Default Rate:NiagaraFloat=&Multiply_Float(
     A=100
     B=/VehicleCosmetics/FX/Niagara/Modules/VC_Mask_Floats(Coro Value:NiagaraFloat=1, FrontEnd Value:NiagaraFloat=0.75, FN Value:NiagaraFloat=0.5)
    ))
   &VC_SpawnRibbon_v1(Number of Ribbons=$Emitter.PortCount, SpawnRate=$Emitter.Spawn Rate)
   /SetVariables(
    Emitter.Dissolve UV01:Vector3f=&VC_Mask_Vectors(
     Coro Value:Vector3f=<1 1 0>
     FrontEnd Value:Vector3f=&MakeCustomVectorFromBool(Bool=$User.bIsFrontendPreview, True Vector=<1 1 -2>, False Vector=<2 1 -2>)
     FN Value:Vector3f=<1 1 0>
    )
    Emitter.Dissolve UV02:Vector3f=&VC_Mask_Vectors(
     Coro Value:Vector3f=<1 1 0>
     FrontEnd Value:Vector3f=&MakeCustomVectorFromBool(Bool=$User.bIsFrontendPreview, True Vector=<1 1 -1.5>, False Vector=<2 1 -1.5>)
     FN Value:Vector3f=<1 1 0>
    )
    Emitter.Dissolve UV03:Vector3f=&VC_Mask_Vectors(
     Coro Value:Vector3f=<1 1 0>
     FrontEnd Value:Vector3f=&MakeCustomVectorFromBool(Bool=$User.bIsFrontendPreview, True Vector=<0.3 1 -1.5>, False Vector=<0.8 1 -1.5>)
     FN Value:Vector3f=<1 1 0>
    )
    Emitter.Dissolve UV04:Vector3f=&VC_Mask_Vectors(
     Coro Value:Vector3f=<1 1 0>
     FrontEnd Value:Vector3f=&MakeCustomVectorFromBool(Bool=$User.bIsFrontendPreview, True Vector=<1 1 -2>, False Vector=<2 1 -2>)
     FN Value:Vector3f=<1 1 0>
    )
    Emitter.Dissolve UV05:Vector3f=&VC_Mask_Vectors(
     Coro Value:Vector3f=<1 1 0>
     FrontEnd Value:Vector3f=&MakeCustomVectorFromBool(Bool=$User.bIsFrontendPreview, True Vector=<1 1 -2>, False Vector=<1 1 -2>)
     FN Value:Vector3f=<1 1 0>
    )
    Emitter.Dissolve UV06:Vector3f=&VC_Mask_Vectors(
     Coro Value:Vector3f=<1 1 0>
     FrontEnd Value:Vector3f=&MakeCustomVectorFromBool(Bool=$User.bIsFrontendPreview, True Vector=<0.3 1 -1.5>, False Vector=<0.8 1 -1.5>)
     FN Value:Vector3f=<1 1 0>
    )
   ) {isSetParameters=true}
  ]
  particleSpawnScript [
   &VC_PaintedVariantsOverride() {enabled=false}
   &VC_InitializeRibbon_v1(Color=<rgba 0 0 0 1>, Enable Skeletal Position=true, Lifetime=1.5, Offset Axis:ENiagaraOrientationAxis=::XAxis, Ribbon Width=50)
   &AddVelocity(Velocity=&VC_Mask_Vectors(Coro Value:Vector3f=<-400 0 0>, FrontEnd Value:Vector3f=<-1200 0 0>, FN Value:Vector3f=<-1200 0 0>)) {module=AddVelocity001}
   /VehicleCosmetics/FX/VC_FXParents/Modules/DLR_ProductPreviews(GameModeID:NiagaraInt32=$System.GameModeID, bPreviewVehicleSpeed:NiagaraBool=$System.bPreviewVehicleSpeed)
  ]
  particleUpdateScript [
   &ParticleState()
   &InheritSourceMovement(
    Applied Position Scale=&VectorFromFloat(Value=&FloatFromCurve(FloatCurve=DataInterface() {propertyValues {Curve {
         defaultValue=1
         keys [
          {arriveTangent=-1.7885, interpMode=RCIM_Cubic, leaveTangent=-1.7885, tangentMode=RCTM_User, time=0.05, value=1}
          {interpMode=RCIM_Cubic, tangentMode=RCTM_User, time=1, value=0}
         ]
        }}}))
    ApplyForce=false
    ApplyPosition=true
    ApplyPositionFalloff=false
   )
   &SolveForcesAndVelocity(Clamp Velocity=false, Limit Acceleration=false)
   &VC_UpdateRibbon_v1(Detach Locked Particle=true, Enable Skeletal Position=true)
   /SetVariables(
    Particles.R1Width:NiagaraFloat=&Multiply_Float(
     A=&SelectFloatFromArray(Direct Array Index=&SpawnGroup(), Float Selection Array=DataInterface() {propertyValues {FloatData [0,60]}})
     B=&FloatFromCurve(
      FloatCurve=DataInterface() {propertyValues {Curve {keys [<at 0.3 1>,<at 1 0.5>]}}}
      Scale Curve=&VC_Mask_Floats(
       Coro Value:NiagaraFloat=1
       FrontEnd Value:NiagaraFloat=&MakeCustomFloatFromBool(Bool=$User.bIsFrontendPreview, True Float=0.65, False Float=0.7)
       FN Value:NiagaraFloat=1
      )
     )
    )
    Particles.R2Width:NiagaraFloat=&Multiply_Float(
     A=&SelectFloatFromArray(Direct Array Index=&SpawnGroup(), Float Selection Array=DataInterface() {propertyValues {FloatData [60,0]}})
     B=&FloatFromCurve(
      FloatCurve=DataInterface() {propertyValues {Curve {keys [<at 0.3 1>,<at 1 0.5>]}}}
      Scale Curve=&VC_Mask_Floats(
       Coro Value:NiagaraFloat=1
       FrontEnd Value:NiagaraFloat=&MakeCustomFloatFromBool(Bool=$User.bIsFrontendPreview, True Float=0.65, False Float=0.7)
       FN Value:NiagaraFloat=1
      )
     )
    )
    Particles.R3Width:NiagaraFloat=&Multiply_Float(
     A=&SelectFloatFromArray(Direct Array Index=&SpawnGroup(), Float Selection Array=DataInterface() {propertyValues {FloatData [0,50]}})
     B=&FloatFromCurve(
      FloatCurve=DataInterface() {propertyValues {Curve {keys [<at 0 1>,<at 1 0.8>]}}}
      Scale Curve=&VC_Mask_Floats(
       Coro Value:NiagaraFloat=1
       FrontEnd Value:NiagaraFloat=&MakeCustomFloatFromBool(Bool=$User.bIsFrontendPreview, True Float=0.65, False Float=0.7)
       FN Value:NiagaraFloat=1
      )
     )
    )
    Particles.R4Width:NiagaraFloat=&Multiply_Float(
     A=&SelectFloatFromArray(Direct Array Index=&SpawnGroup(), Float Selection Array=DataInterface() {propertyValues {FloatData [50,0]}})
     B=&FloatFromCurve(
      FloatCurve=DataInterface() {propertyValues {Curve {keys [<at 0 1>,<at 1 0.8>]}}}
      Scale Curve=&VC_Mask_Floats(
       Coro Value:NiagaraFloat=1
       FrontEnd Value:NiagaraFloat=&MakeCustomFloatFromBool(Bool=$User.bIsFrontendPreview, True Float=0.65, False Float=0.7)
       FN Value:NiagaraFloat=1
      )
     )
    )
    Particles.R5Width:NiagaraFloat=&Multiply_Float(
     A=&SelectFloatFromArray(Direct Array Index=&SpawnGroup(), Float Selection Array=DataInterface() {propertyValues {FloatData [0,25]}})
     B=&FloatFromCurve(
      FloatCurve=DataInterface() {propertyValues {Curve {keys [<at 0 1>,<at 1 0.3>]}}}
      Scale Curve=&VC_Mask_Floats(
       Coro Value:NiagaraFloat=1
       FrontEnd Value:NiagaraFloat=&MakeCustomFloatFromBool(Bool=$User.bIsFrontendPreview, True Float=0.7, False Float=0.3)
       FN Value:NiagaraFloat=1
      )
     )
    )
    Particles.R6Width:NiagaraFloat=&Multiply_Float(
     A=&SelectFloatFromArray(Direct Array Index=&SpawnGroup(), Float Selection Array=DataInterface() {propertyValues {FloatData [25,0]}})
     B=&FloatFromCurve(
      FloatCurve=DataInterface() {propertyValues {Curve {keys [<at 0 1>,<at 1 0.3>]}}}
      Scale Curve=&VC_Mask_Floats(
       Coro Value:NiagaraFloat=1
       FrontEnd Value:NiagaraFloat=&MakeCustomFloatFromBool(Bool=$User.bIsFrontendPreview, True Float=0.7, False Float=0.3)
       FN Value:NiagaraFloat=1
      )
     )
    )
   ) {isSetParameters=true}
   /SetVariables(
    Particles.Color1:LinearColor=<rgba 0 0 0 15>
    Particles.Color2=/VehicleCosmetics/FX/Niagara/Modules/VC_Mask_LinearColor(
     Coro Color:LinearColor=<rgba 0 0 0 0>
     Coro FrontEnd Color:LinearColor=<rgba 0 0 0 0>
     FN Color:LinearColor=&MakeCustomLinearColorFromBool(Bool=$User.bIsFrontendPreview, True Color=<rgba 1 0.1302 0.2501 10>, False Color=<rgba 1 0.0976 0.3162 10>)
    )
    Particles.Color3:LinearColor=&VC_Mask_LinearColor(
     Coro Color:LinearColor=<rgba 0 0 0 1>
     Coro FrontEnd Color:LinearColor=<rgba 0 0 0 1>
     FN Color:LinearColor=&MakeCustomLinearColorFromBool(Bool=$User.bIsFrontendPreview, True Color=<rgba 10 7.6483 3.3333 1>, False Color=<rgba 7 4.2308 1.1667 1>)
    )
   ) {isSetParameters=true}
   /SetVariables(
    Particles.R1Position:NiagaraPosition=&AddVectorToPosition(Position=<0 -10 0>, Vector=Linked() {linkedVariable {name=Particles.Position, type=NiagaraPosition}})
    Particles.R2Position:NiagaraPosition=&AddVectorToPosition(Position=<0 10 0>, Vector=Linked() {linkedVariable {name=Particles.Position, type=NiagaraPosition}})
    Particles.R3Position:NiagaraPosition=&AddVectorToPosition(Position=<0 -5 0>, Vector=Linked() {linkedVariable {name=Particles.Position, type=NiagaraPosition}})
    Particles.R4Position:NiagaraPosition=&AddVectorToPosition(Position=<0 5 0>, Vector=Linked() {linkedVariable {name=Particles.Position, type=NiagaraPosition}})
   ) {isSetParameters=true}
  ]
  rendererClasses [&NiagaraRibbonRendererProperties]
  renderers [
   {rendererClass=&NiagaraRibbonRendererProperties}
   {rendererClass=&NiagaraRibbonRendererProperties, rendererIndex=1}
   {rendererClass=&NiagaraRibbonRendererProperties, rendererIndex=2}
   {rendererClass=&NiagaraRibbonRendererProperties, rendererIndex=3}
   {rendererClass=&NiagaraRibbonRendererProperties, rendererIndex=4}
   {rendererClass=&NiagaraRibbonRendererProperties, rendererIndex=5}
  ]
 }
 NE_VC_Boost_Sprite {
  emitterUpdateScript [
   &EmitterState()
   &VC_FilteredSocketCount()
   &SpawnRate(
    SpawnRate=&Multiply_Float(
     A=50
     B=&VC_Mask_Floats(
      Coro Value:NiagaraFloat=1
      FrontEnd Value:NiagaraFloat=&MakeCustomFloatFromBool(Bool=$User.bIsFrontendPreview, True Float=0.4, False Float=1)
      FN Value:NiagaraFloat=0.5
     )
    )
    Use Spawn Probability=false
   )
   &SpawnRate(
    SpawnGroup=1
    SpawnRate=&Multiply_Float(
     A=50
     B=&VC_Mask_Floats(
      Coro Value:NiagaraFloat=1
      FrontEnd Value:NiagaraFloat=&MakeCustomFloatFromBool(Bool=$User.bIsFrontendPreview, True Float=0.15, False Float=0.8)
      FN Value:NiagaraFloat=0.5
     )
    )
    Use Spawn Probability=false
   ) {module=SpawnRate001}
   &SpawnRate(
    SpawnGroup=2
    SpawnRate=&Multiply_Float(
     A=100
     B=&VC_Mask_Floats(
      Coro Value:NiagaraFloat=1
      FrontEnd Value:NiagaraFloat=&MakeCustomFloatFromBool(Bool=$User.bIsFrontendPreview, True Float=0.5, False Float=0.8)
      FN Value:NiagaraFloat=0.5
     )
    )
    Use Spawn Probability=false
   ) {module=SpawnRate002}
   &SpawnRate(
    SpawnGroup=3
    SpawnRate=&Multiply_Float(
     A=100
     B=&VC_Mask_Floats(
      Coro Value:NiagaraFloat=1
      FrontEnd Value:NiagaraFloat=&MakeCustomFloatFromBool(Bool=$User.bIsFrontendPreview, True Float=0.25, False Float=0.8)
      FN Value:NiagaraFloat=0.5
     )
    )
    Use Spawn Probability=false
   ) {module=SpawnRate003}
  ]
  particleSpawnScript [
   &VC_PaintedVariantsOverride() {enabled=false}
   &InitializeParticle(
    Color=&MergeHelper_ColorSwitch(
     0=<rgba 1 0.4115 0.9663 1>
     1=&LerpLinearColors(
      StartColor=<rgba 1 0.7423 0.207 1>
      EndColor=<rgba 0.4 0.2969 0.0828 1>
      LerpFactor=&UniformAOrBFloat(A=0, Evaluation Type=::NewEnumerator0, Recalculate Random Each Loop=true, Override Seed=false)
     )
     2=<rgba 1 1 1 1>
     3=<rgba 1 1 1 1>
     Index=&SpawnGroup()
     Num=::NewEnumerator2
    )
    Lifetime=&MergeHelper_FloatSwitch(
     0=&VC_Mask_Floats(
      Coro Value:NiagaraFloat=1
      FrontEnd Value:NiagaraFloat=&MakeCustomFloatFromBool(
       Bool=$User.bIsFrontendPreview
       True Float=&RandomRangeFloat(Minimum=0.2, Maximum=0.3)
       False Float=&RandomRangeFloat(Minimum=0.25, Maximum=0.15)
      )
      FN Value:NiagaraFloat=0.5
     )
     1=&VC_Mask_Floats(
      Coro Value:NiagaraFloat=1
      FrontEnd Value:NiagaraFloat=&MakeCustomFloatFromBool(
       Bool=$User.bIsFrontendPreview
       True Float=&RandomRangeFloat(Minimum=0.15, Maximum=0.25)
       False Float=&RandomRangeFloat(Minimum=0.25, Maximum=0.15)
      )
      FN Value:NiagaraFloat=0.5
     )
     2=&RandomRangeFloat(Minimum=0.5, Maximum=0.7)
     3=&RandomRangeFloat(Minimum=0.5, Maximum=0.7)
     Index=&SpawnGroup()
     Num=::NewEnumerator2
    )
    Mesh Scale=&MergeHelper_VectorSwitch(
     0=&VC_Mask_Vectors(
      Coro Value:Vector3f=<-400 0 0>
      FrontEnd Value:Vector3f=&MakeCustomVectorFromBool(Bool=$User.bIsFrontendPreview, True Vector=<4 5 4>, False Vector=<3 4.5 3>)
      FN Value:Vector3f=<-1200 0 0>
     )
     1=<2.5 4 2.5>
     2=<0 0 0>
     3=<0 0 0>
     Index=&SpawnGroup()
     Num=::NewEnumerator2
    )
    Mesh Scale Mode=::NewEnumerator4
    Position Mode=::NewEnumerator0
    Sprite Rotation Mode=::NewEnumerator2
    Sprite Size Mode=::NewEnumerator3
    Uniform Sprite Size=&MergeHelper_FloatSwitch(0=0, 1=0, 2=135, 3=135, Index=&SpawnGroup(), Num=::NewEnumerator2)
    Write Material Random=true
   )
   &PartitionParticles(Number Of Partitions=$Emitter.PortCount)
   &SkeletalMeshLocation(
    Bone / Socket Index=$Particles.PartitionParticles.Partition
    Bone / Socket Sampling Mode=::NewEnumerator5
    MaskSpawnGroup=false
    Mesh Sampling Type=::NewEnumerator6
    Sampled Position Offset=&AddVectorToPosition(
     Position=&VC_Mask_Vectors(
      Coro Value:Vector3f=<0 0 0>
      FrontEnd Value:Vector3f=&MakeCustomVectorFromBool(Bool=$User.bIsFrontendPreview, True Vector=<0 0 0>, False Vector=<70 0 0>)
      FN Value:Vector3f=<0 0 0>
     )
     Vector=Linked() {linkedVariable {name=User.DTCarOffset, type=NiagaraPosition}}
    )
    Skeletal Mesh=$System.SkeletalMesh
   )
   &AddVelocity(Velocity=&VC_Mask_Vectors(
     Coro Value:Vector3f=<-400 0 0>
     FrontEnd Value:Vector3f=&MakeCustomVectorFromBool(
      Bool=$User.bIsFrontendPreview
      True Vector=&MergeHelper_VectorSwitch(0=<-10 0 0>, 1=<-10 0 0>, 2=<-270 0 0>, 3=<-270 0 0>, Index=&SpawnGroup(), Num=::NewEnumerator2)
      False Vector=&MergeHelper_VectorSwitch(0=<0 0 0>, 1=<0 0 0>, 2=<-300 0 0>, 3=<-300 0 0>, Index=&SpawnGroup(), Num=::NewEnumerator2)
     )
     FN Value:Vector3f=<-1200 0 0>
    ))
   &DLR_ProductPreviews(GameModeID:NiagaraInt32=$System.GameModeID, bPreviewVehicleSpeed:NiagaraBool=$System.bPreviewVehicleSpeed)
   &InitialMeshOrientation(
    Mesh Orientation Mode=::NewEnumerator2
    Orientation Coordinate Space=::Local
    Orientation Vector=<0 -1 0>
    Rotation=&RandomRangeVector(Minimum=<1 0 0>, Maximum=<-1 0 0>)
    Rotation Coordinate Space=::NewEnumerator2
    Use Rotation Vector=true
   )
   /SetVariables(Particles.VisibilityTag=&SelectIntFromArray(Direct Array Index=&SpawnGroup(), Int Selection Array=DataInterface() {propertyValues {IntData [0,0,1,2]}})) {isSetParameters=true}
   &CameraOffset(Camera Offset Amount=&MergeHelper_FloatSwitch(0=0, 1=-30, 2=0, 3=0, Index=&SpawnGroup(), Num=::NewEnumerator2))
  ]
  particleUpdateScript [
   &ParticleState()
   &SpriteFacingAndAlignment(Alignment Coordinate Space=::Local, Facing Coordinate Space=::Local, WriteAlignment=true, WriteFacing=true)
   &ScaleSpriteSize(Uniform Curve Sprite Scale=DataInterface() {propertyValues {Curve {keys [
        {arriveTangent=0.4468, interpMode=RCIM_Cubic, leaveTangent=0.4468, tangentMode=RCTM_User, time=0, value=0.2}
        {arriveTangent=0.564, interpMode=RCIM_Cubic, leaveTangent=0.564, tangentMode=RCTM_User, time=0.5, value=0.4}
        {arriveTangent=1.0034, interpMode=RCIM_Cubic, leaveTangent=1.0034, tangentMode=RCTM_User, time=1, value=1}
       ]}}})
   &InheritSourceMovement(
    Applied Position Scale=&VectorFromFloat(Value=&FloatFromCurve(FloatCurve=DataInterface() {propertyValues {Curve {
         defaultValue=1
         keys [
          {arriveTangent=-1.7885, interpMode=RCIM_Cubic, leaveTangent=-1.7885, tangentMode=RCTM_User, time=0.05, value=1}
          {interpMode=RCIM_Cubic, tangentMode=RCTM_User, time=1, value=0}
         ]
        }}}))
    ApplyForce=false
    ApplyPosition=true
    ApplyPositionFalloff=false
   ) {enabled=false}
   &ScaleMeshSize(Scale Factor=&VectorFromFloat(Value=&FloatFromCurve(FloatCurve=DataInterface() {propertyValues {Curve {keys [{tangentMode=RCTM_User, time=0, value=0}, {tangentMode=RCTM_User, time=0.8, value=0.9}, <at 1 1>]}}})))
   &ScaleColor(
    Scale RGB=&VectorFromFloat(Value=&MergeHelper_FloatSwitch(
      0=&VC_Mask_Floats(
       Coro Value:NiagaraFloat=1
       FrontEnd Value:NiagaraFloat=&MakeCustomFloatFromBool(
        Bool=$User.bIsFrontendPreview
        True Float=&RandomRangeFloat(
         Minimum=&FloatFromCurve(FloatCurve=DataInterface() {propertyValues {Curve {keys [<at 0.5 0.5>,<at 1 2>]}}})
         Maximum=&FloatFromCurve(FloatCurve=DataInterface() {propertyValues {Curve {keys [<at 0 1>,<at 0.7 0.5>]}}})
        )
        False Float=&RandomRangeFloat(
         Minimum=&FloatFromCurve(FloatCurve=DataInterface() {propertyValues {Curve {keys [<at 0.5 0.5>,<at 1 2>]}}})
         Maximum=&FloatFromCurve(FloatCurve=DataInterface() {propertyValues {Curve {keys [<at 0 1>,<at 0.7 0.5>]}}})
        )
       )
       FN Value:NiagaraFloat=0.5
      )
      1=&VC_Mask_Floats(
       Coro Value:NiagaraFloat=1
       FrontEnd Value:NiagaraFloat=&MakeCustomFloatFromBool(
        Bool=$User.bIsFrontendPreview
        True Float=&RandomRangeFloat(
         Minimum=&FloatFromCurve(FloatCurve=DataInterface() {propertyValues {Curve {keys [<at 0.5 0.4>,<at 1 2>]}}})
         Maximum=&FloatFromCurve(FloatCurve=DataInterface() {propertyValues {Curve {keys [<at 0 1>,<at 0.7 0.4>]}}})
        )
        False Float=&RandomRangeFloat(
         Minimum=&FloatFromCurve(FloatCurve=DataInterface() {propertyValues {Curve {keys [<at 0.5 0.5>,<at 1 2>]}}})
         Maximum=&FloatFromCurve(FloatCurve=DataInterface() {propertyValues {Curve {keys [<at 0 1>,<at 0.7 0.5>]}}})
        )
       )
       FN Value:NiagaraFloat=0.5
      )
      2=&VC_Mask_Floats(
       Coro Value:NiagaraFloat=1
       FrontEnd Value:NiagaraFloat=&MakeCustomFloatFromBool(
        Bool=$User.bIsFrontendPreview
        True Float=&RandomRangeFloat(
         Minimum=&FloatFromCurve(FloatCurve=DataInterface() {propertyValues {Curve {keys [<at 0.5 0.1>,<at 1 1>]}}})
         Maximum=&FloatFromCurve(FloatCurve=DataInterface() {propertyValues {Curve {keys [<at 0 1>,<at 0.7 0.1>]}}})
        )
        False Float=&RandomRangeFloat(
         Minimum=&FloatFromCurve(FloatCurve=DataInterface() {propertyValues {Curve {keys [<at 0.5 0.1>,<at 1 1>]}}})
         Maximum=&FloatFromCurve(FloatCurve=DataInterface() {propertyValues {Curve {keys [<at 0 1>,<at 0.7 0.1>]}}})
        )
       )
       FN Value:NiagaraFloat=0.5
      )
      3=&VC_Mask_Floats(
       Coro Value:NiagaraFloat=1
       FrontEnd Value:NiagaraFloat=&MakeCustomFloatFromBool(
        Bool=$User.bIsFrontendPreview
        True Float=&RandomRangeFloat(
         Minimum=&FloatFromCurve(FloatCurve=DataInterface() {propertyValues {Curve {keys [<at 0.5 0.1>,<at 1 1>]}}})
         Maximum=&FloatFromCurve(FloatCurve=DataInterface() {propertyValues {Curve {keys [<at 0 1>,<at 0.7 0.1>]}}})
        )
        False Float=&RandomRangeFloat(
         Minimum=&FloatFromCurve(FloatCurve=DataInterface() {propertyValues {Curve {keys [<at 0.5 0.1>,<at 1 1>]}}})
         Maximum=&FloatFromCurve(FloatCurve=DataInterface() {propertyValues {Curve {keys [<at 0 1>,<at 0.7 0.1>]}}})
        )
       )
       FN Value:NiagaraFloat=0.5
      )
      Index=&SpawnGroup()
      Num=::NewEnumerator2
     ))
    ScaleA=true
    ScaleRGB=true
   )
   &DynamicMaterialParameters(
    Index 0 Param 2=1
    Module.Index0Param0Write=true
    Module.Index0Param1Write=true
    Module.Index0Param2Write=true
    Module.Index0Param3Write=true
   )
   &MeshRotationForce(Rotation=&MergeHelper_VectorSwitch(
     0=<0 0 0>
     1=<0 0 0>
     2=&RandomRangeVector(Minimum=<-1 0 0>, Maximum=<1 0 0>)
     3=&RandomRangeVector(Minimum=<-1 0 0>, Maximum=<1 0 0>)
     Index=&SpawnGroup()
     Num=::NewEnumerator2
    ))
   &SolveForcesAndVelocity(Clamp Velocity=false, Limit Acceleration=false)
   &DynamicMaterialParameters(
    Index 0 Param 1=&MergeHelper_FloatSwitch(0=0.2, 1=0.2, 2=0, 3=0, Index=&SpawnGroup(), Num=::NewEnumerator2)
    Index 0 Param 2=&MergeHelper_FloatSwitch(
     0=&RandomRangeFloat(Minimum=4, Maximum=8)
     1=&RandomRangeFloat(Minimum=4, Maximum=8)
     2=0
     3=0
     Index=&SpawnGroup()
     Num=::NewEnumerator2
    )
    Index 0 Param 3=&MergeHelper_FloatSwitch(0=1, 1=1, 2=0, 3=0, Index=&SpawnGroup(), Num=::NewEnumerator2)
    Module.Index0Float0Write=true
    Module.Index0Float1Write=true
    Module.Index0Float2Write=true
    Module.Index0Float3Write=true
   ) {module=DynamicMaterialParameters001}
   /SetVariables(Particles.SubImageIndex=&MergeHelper_FloatSwitch(
     0=0
     1=0
     2=&FloatFromCurve(FloatCurve=DataInterface() {propertyValues {Curve {keys [
          {arriveTangent=9.5406, interpMode=RCIM_Cubic, leaveTangent=9.5406, tangentMode=RCTM_User, time=0, value=0}
          {arriveTangent=28.2845, interpMode=RCIM_Cubic, leaveTangent=28.2845, tangentMode=RCTM_User, time=1, value=16}
         ]}}})
     3=&FloatFromCurve(FloatCurve=DataInterface() {propertyValues {Curve {keys [
          {arriveTangent=9.5406, interpMode=RCIM_Cubic, leaveTangent=9.5406, tangentMode=RCTM_User, time=0, value=0}
          {arriveTangent=28.2845, interpMode=RCIM_Cubic, leaveTangent=28.2845, tangentMode=RCTM_User, time=1, value=16}
         ]}}})
     Index=&SpawnGroup()
     Num=::NewEnumerator2
    )) {isSetParameters=true}
  ]
  rendererClasses [&NiagaraMeshRendererProperties, &NiagaraSpriteRendererProperties]
  renderers [
   {rendererClass=&NiagaraMeshRendererProperties, rendererIndex=0}
   {rendererClass=&NiagaraSpriteRendererProperties, rendererIndex=1}
   {rendererClass=&NiagaraSpriteRendererProperties, rendererIndex=2}
  ]
 }
 NE_VC_Boost_Twist {
  emitterUpdateScript [
   &EmitterState()
   &VC_FilteredSocketCount()
   &VC_RateQualityHandler(Default Rate:NiagaraFloat=&Multiply_Float(
     A=100
     B=&VC_Mask_Floats(Coro Value:NiagaraFloat=1, FrontEnd Value:NiagaraFloat=0.75, FN Value:NiagaraFloat=0.5)
    ))
   &VC_SpawnRibbon_v1(Number of Ribbons=$Emitter.PortCount, SpawnRate=$Emitter.Spawn Rate)
   /SetVariables(
    Emitter.WPO Speed01:NiagaraFloat=&VC_Mask_Floats(
     Coro Value:NiagaraFloat=1
     FrontEnd Value:NiagaraFloat=&MakeCustomFloatFromBool(Bool=$User.bIsFrontendPreview, True Float=1.7, False Float=2)
     FN Value:NiagaraFloat=0.5
    )
    Emitter.WPO Speed02:NiagaraFloat=&VC_Mask_Floats(
     Coro Value:NiagaraFloat=1
     FrontEnd Value:NiagaraFloat=&MakeCustomFloatFromBool(Bool=$User.bIsFrontendPreview, True Float=-1.7, False Float=-2)
     FN Value:NiagaraFloat=0.5
    )
    Emitter.WPO Speed03:NiagaraFloat=&VC_Mask_Floats(
     Coro Value:NiagaraFloat=1
     FrontEnd Value:NiagaraFloat=&MakeCustomFloatFromBool(Bool=$User.bIsFrontendPreview, True Float=-1.7, False Float=-2)
     FN Value:NiagaraFloat=0.5
    )
    Emitter.WPO Speed04:NiagaraFloat=&VC_Mask_Floats(
     Coro Value:NiagaraFloat=1
     FrontEnd Value:NiagaraFloat=&MakeCustomFloatFromBool(Bool=$User.bIsFrontendPreview, True Float=1.7, False Float=2)
     FN Value:NiagaraFloat=0.5
    )
   ) {isSetParameters=true}
   /SetVariables(
    Emitter.WPO Int01:NiagaraFloat=&VC_Mask_Floats(
     Coro Value:NiagaraFloat=1
     FrontEnd Value:NiagaraFloat=&MakeCustomFloatFromBool(Bool=$User.bIsFrontendPreview, True Float=14, False Float=15)
     FN Value:NiagaraFloat=1
    )
    Emitter.WPO Int02:NiagaraFloat=&VC_Mask_Floats(
     Coro Value:NiagaraFloat=1
     FrontEnd Value:NiagaraFloat=&MakeCustomFloatFromBool(Bool=$User.bIsFrontendPreview, True Float=11, False Float=12)
     FN Value:NiagaraFloat=1
    )
    Emitter.WPO Int03:NiagaraFloat=&VC_Mask_Floats(
     Coro Value:NiagaraFloat=1
     FrontEnd Value:NiagaraFloat=&MakeCustomFloatFromBool(Bool=$User.bIsFrontendPreview, True Float=14, False Float=15)
     FN Value:NiagaraFloat=1
    )
    Emitter.WPO Int04:NiagaraFloat=&VC_Mask_Floats(
     Coro Value:NiagaraFloat=1
     FrontEnd Value:NiagaraFloat=&MakeCustomFloatFromBool(Bool=$User.bIsFrontendPreview, True Float=11, False Float=12)
     FN Value:NiagaraFloat=1
    )
   ) {isSetParameters=true}
   /SetVariables(
    Emitter.WPO Tile01:NiagaraFloat=&VC_Mask_Floats(
     Coro Value:NiagaraFloat=1
     FrontEnd Value:NiagaraFloat=&MakeCustomFloatFromBool(Bool=$User.bIsFrontendPreview, True Float=-0.8, False Float=-0.5)
     FN Value:NiagaraFloat=0.5
    )
    Emitter.WPO Tile02:NiagaraFloat=&VC_Mask_Floats(
     Coro Value:NiagaraFloat=1
     FrontEnd Value:NiagaraFloat=&MakeCustomFloatFromBool(Bool=$User.bIsFrontendPreview, True Float=0.8, False Float=0.5)
     FN Value:NiagaraFloat=0.5
    )
    Emitter.WPO Tile03:NiagaraFloat=&VC_Mask_Floats(
     Coro Value:NiagaraFloat=1
     FrontEnd Value:NiagaraFloat=&MakeCustomFloatFromBool(Bool=$User.bIsFrontendPreview, True Float=0.8, False Float=0.5)
     FN Value:NiagaraFloat=0.5
    )
    Emitter.WPO Tile04:NiagaraFloat=&VC_Mask_Floats(
     Coro Value:NiagaraFloat=1
     FrontEnd Value:NiagaraFloat=&MakeCustomFloatFromBool(Bool=$User.bIsFrontendPreview, True Float=-0.8, False Float=-0.5)
     FN Value:NiagaraFloat=0.5
    )
   ) {isSetParameters=true}
  ]
  particleSpawnScript [
   &VC_PaintedVariantsOverride() {enabled=false}
   &VC_InitializeRibbon_v1(Enable Skeletal Position=true, Lifetime=1.5)
   &AddVelocity(Velocity=&VC_Mask_Vectors(Coro Value:Vector3f=<-400 0 0>, FrontEnd Value:Vector3f=<-1200 0 0>, FN Value:Vector3f=<-1200 0 0>)) {module=AddVelocity001}
   &DLR_ProductPreviews(GameModeID:NiagaraInt32=$System.GameModeID, bPreviewVehicleSpeed:NiagaraBool=$System.bPreviewVehicleSpeed)
  ]
  particleUpdateScript [
   &ParticleState()
   &InheritSourceMovement(
    Applied Position Scale=&VectorFromFloat(Value=&FloatFromCurve(FloatCurve=DataInterface() {propertyValues {Curve {
         defaultValue=1
         keys [
          {arriveTangent=-1.7885, interpMode=RCIM_Cubic, leaveTangent=-1.7885, tangentMode=RCTM_User, time=0.05, value=1}
          {interpMode=RCIM_Cubic, tangentMode=RCTM_User, time=1, value=0}
         ]
        }}}))
    ApplyForce=false
    ApplyPosition=true
    ApplyPositionFalloff=false
   )
   /SetVariables(Particles.Ribbon_BlackColor:LinearColor=<rgba 0 0 0 10>, Particles.Ribbon_Yellow:LinearColor=<rgba 50 21.3777 5.5433 1>) {isSetParameters=true}
   &SolveForcesAndVelocity(Clamp Velocity=false, Limit Acceleration=false)
   &VC_UpdateRibbon_v1(Enable Skeletal Position=true)
   /SetVariables(
    Particles.R1Width:NiagaraFloat=&Multiply_Float(
     A=&SelectFloatFromArray(Direct Array Index=&SpawnGroup(), Float Selection Array=DataInterface() {propertyValues {FloatData [0,45]}})
     B=&FloatFromCurve(FloatCurve=DataInterface() {propertyValues {Curve {keys [<at 0 1>,<at 1 0>]}}}, Scale Curve=0.5)
    )
    Particles.R2Width:NiagaraFloat=&Multiply_Float(
     A=&SelectFloatFromArray(Direct Array Index=&SpawnGroup(), Float Selection Array=DataInterface() {propertyValues {FloatData [45,0]}})
     B=&FloatFromCurve(FloatCurve=DataInterface() {propertyValues {Curve {keys [<at 0 1>,<at 1 0>]}}}, Scale Curve=0.5)
    )
    Particles.R3Width:NiagaraFloat=&Multiply_Float(
     A=&SelectFloatFromArray(Direct Array Index=&SpawnGroup(), Float Selection Array=DataInterface() {propertyValues {FloatData [0,35]}})
     B=&FloatFromCurve(
      FloatCurve=DataInterface() {propertyValues {Curve {keys [<at 0 1>,<at 1 0>]}}}
      Scale Curve=&VC_Mask_Floats(
       Coro Value:NiagaraFloat=1
       FrontEnd Value:NiagaraFloat=&MakeCustomFloatFromBool(Bool=$User.bIsFrontendPreview, True Float=0.5, False Float=1)
       FN Value:NiagaraFloat=0.5
      )
     )
    )
    Particles.R4Width:NiagaraFloat=&Multiply_Float(
     A=&SelectFloatFromArray(Direct Array Index=&SpawnGroup(), Float Selection Array=DataInterface() {propertyValues {FloatData [35,0]}})
     B=&FloatFromCurve(
      FloatCurve=DataInterface() {propertyValues {Curve {keys [<at 0 1>,<at 1 0>]}}}
      Scale Curve=&VC_Mask_Floats(
       Coro Value:NiagaraFloat=1
       FrontEnd Value:NiagaraFloat=&MakeCustomFloatFromBool(Bool=$User.bIsFrontendPreview, True Float=0.5, False Float=1)
       FN Value:NiagaraFloat=0.5
      )
     )
    )
   ) {isSetParameters=true}
  ]
  rendererClasses [&NiagaraRibbonRendererProperties]
  renderers [
   {rendererClass=&NiagaraRibbonRendererProperties}
   {rendererClass=&NiagaraRibbonRendererProperties, rendererIndex=1}
   {rendererClass=&NiagaraRibbonRendererProperties, rendererIndex=2}
   {rendererClass=&NiagaraRibbonRendererProperties, rendererIndex=3}
  ]
 }
}
summary {
 systemName=NS_Illgraph
 emitters [
  {emitterName=NE_VC_Boost_Twist, rendererClasses [&NiagaraRibbonRendererProperties]}
  {emitterName=NE_VC_Boost_Ribbon, rendererClasses [&NiagaraRibbonRendererProperties]}
  {
   emitterName=NE_VC_Boost_Sprite
   rendererClasses [&NiagaraMeshRendererProperties, &NiagaraSpriteRendererProperties]
  }
 ]
 userVariables [
  User.bIsFrontendPreview {type=NiagaraBool}
  User.bIsPainted {type=NiagaraBool}
  User.blsIsCoroColor {type=NiagaraBool}
  User.bPreviewVehicleSpeed {type=NiagaraBool}
  User.ColorVariantID {type=NiagaraInt32}
  User.DTCarOffset {type=NiagaraPosition, description="", defaultValue {}}
  User.GameModeID {type=NiagaraInt32}
  User.PaintedColor {type=LinearColor, description="", defaultValue {struct=LinearColor, value=<rgba 1 1 1 1>}}
  User.PaintedScalar {type=NiagaraFloat, defaultValue=1, description=""}
  User.PortCount {type=NiagaraInt32, defaultValue=2}
 ]
}
